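# Niagara swing trail for the RobustTorn pickaxe, drawing layered anim-trail ribbons between two weapon bones or sockets.
# User swing-section and last-swing parameters choose the socket pair; the last swing also widens the trail.
# Base and Soft are amber ribbons tinted and faded by time of day, with Base adding a second ribbon renderer and screen-oriented color and width parameters.
# BlackTopTrail and SupportingTrails add shorter-lived black ribbons.
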NiagaraSystem /BRCosmetics/Weapons/FORT_Melee/Pickaxe_RobustTorn/FX/NS_Pickaxe_RobustTorn_Trail
dependencies {
 usedDataInterfaces [&NiagaraDataInterfaceCurve, &NiagaraDataInterfaceSocketReader, &NiagaraDataInterfaceColorCurve]
 usedDynamicInputs [&ColorFromCurve]
 usedModules [
  &self:SystemSpawnScript.SystemScriptSource.SystemScriptGraph.NiagaraNodeAssignment_0.SetVariables_#1
  &SystemState
  &self:SystemSpawnScript.SystemScriptSource.SystemScriptGraph.NiagaraNodeAssignment_1.SetVariables_#2
  &self:SystemSpawnScript.SystemScriptSource.SystemScriptGraph.NiagaraNodeAssignment_2.SetVariables_#3
 ]
}
emitters {
 Base {
  emitterSpawnScript [&FN_RibbonIDOffsetUtility()]
  emitterUpdateScript [
   &EmitterState()
   &FN_SpawnAnimTrail(Spawn Interval=0)
  ]
  particleSpawnScript [
   &FN_InitializeAnimTrailBoneIndecies(
    First Bone or Socket Index:NiagaraInt32=&MakeCustomIntFromBool(
     Bool=$User.IsLastSwing
     True INT=3
     False INT=&MergeHelper_IntSwitch(0=5, 1=1, 2=1, 3=3, 4=?, 5=?, Index=$User.SwingSectionIndex, Num=::NewEnumerator2)
    )
    Second Bone or Socket Index:NiagaraInt32=&MakeCustomIntFromBool(
     Bool=$User.IsLastSwing
     True INT=2
     False INT=&MergeHelper_IntSwitch(0=4, 1=0, 2=2, 3=2, 4=?, 5=?, Index=$User.SwingSectionIndex, Num=::NewEnumerator2)
    )
   )
   &FN_InitializeAnimTrail(
    AnimTrailWidth=&MakeCustomFloatFromBool(Bool=$User.IsLastSwing, True Float=1.2, False Float=1)
    Color=<rgba 0.34 0.1952 0 1>
    Lifetime=0.25
   )
  ]
  particleUpdateScript [
   &FN_UpdateAnimTrail()
   &ScaleColor(
    Scale Alpha=&Multiply_Float(
     A=&FloatFromCurve(
      FloatCurve=DataInterface() {propertyValues {Curve {keys [{interpMode=RCIM_Cubic, leaveTangent=-4.0138, tangentMode=RCTM_Break, time=0, value=1}, <at 0.4 0>]}}}
      Scale Curve=1.5
     )
     B=$System.TODFloat
    )
    Scale RGB=&MakeVectorFromLinearColorRGB(Color=&LerpLinearColors(
      StartColor=$System.TODColor
      EndColor=<rgba 1 1 1 1>
      LerpFactor=&FNTimeOfDayScalar(Night Scale=0, Morning Scale=0.7, Day Scale=0, Evening Scale=0.7)
     ))
    ScaleA=true
    ScaleRGB=true
   ) {module=ScaleColor002}
   /SetVariables(
    Particles.ScreenOrientedTrailColor:LinearColor=&MultiplyLinearColor(Linear Color=$StackContext.Color, Scale Factor=<1 1 1 0.5>)
    Particles.ScreenOrientedTrailWidth:NiagaraFloat=&Multiply_Float(A=$StackContext.RibbonWidth, B=0.6)
   ) {isSetParameters=true}
  ]
  rendererClasses [&NiagaraRibbonRendererProperties]
  renderers [
   {rendererClass=&NiagaraRibbonRendererProperties}
   {rendererClass=&NiagaraRibbonRendererProperties, rendererIndex=1}
  ]
 }
 BlackTopTrail {
  emitterSpawnScript [&FN_RibbonIDOffsetUtility()]
  emitterUpdateScript [
   &EmitterState()
   &FN_SpawnAnimTrail()
  ]
  particleSpawnScript [
   &FN_InitializeAnimTrailBoneIndecies(
    First Bone or Socket Index:NiagaraInt32=&MakeCustomIntFromBool(
     Bool=$User.IsLastSwing
     True INT=3
     False INT=&MergeHelper_IntSwitch(0=5, 1=1, 2=1, 3=3, 4=?, 5=?, Index=$User.SwingSectionIndex, Num=::NewEnumerator2)
    )
    Second Bone or Socket Index:NiagaraInt32=&MakeCustomIntFromBool(
     Bool=$User.IsLastSwing
     True INT=2
     False INT=&MergeHelper_IntSwitch(0=4, 1=0, 2=2, 3=2, 4=?, 5=?, Index=$User.SwingSectionIndex, Num=::NewEnumerator2)
    )
   )
   &FN_InitializeAnimTrail(
    AnimTrailWidth=&MakeCustomFloatFromBool(Bool=$User.IsLastSwing, True Float=1.2, False Float=1)
    Color=<rgba 0 0 0 1>
    Lifetime=0.16
   )
  ]
  particleUpdateScript [
   &FN_UpdateAnimTrail(Locked Particle RibbonWidth=1) {module=FN_UpdateAnimTrail001}
   &ScaleColor(
    Scale Alpha=&FloatFromCurve(FloatCurve=DataInterface() {propertyValues {Curve {keys [{interpMode=RCIM_Cubic, leaveTangent=-4.0138, tangentMode=RCTM_Break, time=0, value=1}, <at 0.4 0>]}}})
    ScaleA=true
    ScaleRGB=true
   )
  ]
  rendererClasses [&NiagaraRibbonRendererProperties]
  renderers [{rendererClass=&NiagaraRibbonRendererProperties}]
 }
 Soft {
  emitterSpawnScript [&FN_RibbonIDOffsetUtility()]
  emitterUpdateScript [
   &EmitterState()
   &FN_SpawnAnimTrail(Spawn Interval=0)
  ]
  particleSpawnScript [
   &FN_InitializeAnimTrailBoneIndecies(
    First Bone or Socket Index:NiagaraInt32=&MakeCustomIntFromBool(
     Bool=$User.IsLastSwing
     True INT=3
     False INT=&MergeHelper_IntSwitch(0=5, 1=1, 2=1, 3=3, 4=?, 5=?, Index=$User.SwingSectionIndex, Num=::NewEnumerator2)
    )
    Second Bone or Socket Index:NiagaraInt32=&MakeCustomIntFromBool(
     Bool=$User.IsLastSwing
     True INT=2
     False INT=&MergeHelper_IntSwitch(0=4, 1=0, 2=2, 3=2, 4=?, 5=?, Index=$User.SwingSectionIndex, Num=::NewEnumerator2)
    )
   )
   &FN_InitializeAnimTrail(
    AnimTrailWidth=&MakeCustomFloatFromBool(Bool=$User.IsLastSwing, True Float=1.2, False Float=1)
    Color=<rgba 0.2 0.1148 0 1>
    Lifetime=0.25
   )
  ]
  particleUpdateScript [
   &FN_UpdateAnimTrail()
   &ScaleColor(
    Scale Alpha=&Multiply_Float(
     A=&FloatFromCurve(
      FloatCurve=DataInterface() {propertyValues {Curve {keys [{interpMode=RCIM_Cubic, leaveTangent=-4.0138, tangentMode=RCTM_Break, time=0, value=1}, <at 0.4 0>]}}}
      Scale Curve=0.8
     )
     B=$System.TODFloat
    )
    Scale RGB=&MakeVectorFromLinearColorRGB(Color=&LerpLinearColors(
      StartColor=$System.TODColor
      EndColor=<rgba 1 1 1 1>
      LerpFactor=&FNTimeOfDayScalar(Night Scale=0, Morning Scale=0.7, Day Scale=0, Evening Scale=0.7)
     ))
    ScaleA=true
    ScaleRGB=true
   ) {module=ScaleColor002}
  ]
  rendererClasses [&NiagaraRibbonRendererProperties]
  renderers [{rendererClass=&NiagaraRibbonRendererProperties}]
 }
 SupportingTrails {
  emitterSpawnScript [&FN_RibbonIDOffsetUtility()]
  emitterUpdateScript [
   &EmitterState()
   &FN_SpawnAnimTrail()
  ]
  particleSpawnScript [
   &FN_InitializeAnimTrailBoneIndecies(
    First Bone or Socket Index:NiagaraInt32=&MakeCustomIntFromBool(
     Bool=$User.IsLastSwing
     True INT=3
     False INT=&MergeHelper_IntSwitch(0=5, 1=1, 2=1, 3=3, 4=?, 5=?, Index=$User.SwingSectionIndex, Num=::NewEnumerator2)
    )
    Second Bone or Socket Index:NiagaraInt32=&MakeCustomIntFromBool(
     Bool=$User.IsLastSwing
     True INT=2
     False INT=&MergeHelper_IntSwitch(0=4, 1=0, 2=2, 3=2, 4=?, 5=?, Index=$User.SwingSectionIndex, Num=::NewEnumerator2)
    )
   )
   &FN_InitializeAnimTrail(
    AnimTrailWidth=&MakeCustomFloatFromBool(Bool=$User.IsLastSwing, True Float=1.2, False Float=1)
    Color=<rgba 0 0 0 1>
    Lifetime=0.16
   )
  ]
  particleUpdateScript [
   &FN_UpdateAnimTrail(Locked Particle RibbonWidth=1) {module=FN_UpdateAnimTrail001}
   &ScaleColor(
    Scale Alpha=&FloatFromCurve(FloatCurve=DataInterface() {propertyValues {Curve {keys [
         {interpMode=RCIM_Cubic, tangentMode=RCTM_User, time=0, value=1}
         {arriveTangent=-5.0049, interpMode=RCIM_Cubic, leaveTangent=-5.0049, tangentMode=RCTM_User, time=0.5, value=0}
        ]}}})
    ScaleA=true
    ScaleRGB=false
   )
  ]
  rendererClasses [&NiagaraRibbonRendererProperties]
  renderers [{rendererClass=&NiagaraRibbonRendererProperties}]
 }
}
summary {
 systemName=NS_Pickaxe_RobustTorn_Trail
 emitters [
  {emitterName=Soft, rendererClasses [&NiagaraRibbonRendererProperties]}
  {emitterName=SupportingTrails, rendererClasses [&NiagaraRibbonRendererProperties]}
  {emitterName=BlackTopTrail, rendererClasses [&NiagaraRibbonRendererProperties]}
  {emitterName=Base, rendererClasses [&NiagaraRibbonRendererProperties]}
 ]
 userVariables [
  User.IsLastSwing {type=NiagaraBool}
  User.RibbonIDOffset {type=NiagaraInt32}
  User.SwingSectionIndex {type=NiagaraInt32}
 ]
}
